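# Material function that computes custom sky-atmosphere-driven lighting for the goo in the Crystal Globe backpack cosmetic from a world-space normal input.
# Specular raises the remapped dot of light direction and reflection vector to the inverse of the highlight parameter's alpha, tinted by color and sun illuminance.
# Diffuse is a wrapped, powered N·L times a color, without sun illuminance.
# Both are summed with ambient-tinted distant sky scattered luminance.
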MaterialFunction /BRCosmetics/Accessories/FORT_Backpacks/Backpack_M_MED_CrystalGlobe/FX/Materials/MF_Backpack_CrystalGlobe_Goo_Shading
graph {nodes [
  n1 = SkyAtmosphereLightDirection()
  n2 = FunctionInput() {inputName=WS Normal}
  n3 = ReflectionVectorWS(CustomWorldNormal=@n2)
  n4 = VectorParameter() {parameterName=Specular Highlight, defaultValue=<rgba 0.1 0.08 0.08 0.0133>}
  n5 = Power(Base=ConstantBiasScale(None=DotProduct(A=@n1, B=@n3)), Exp=Divide(B=@n4.A))
  n6 = Multiply(A=@n4.RGB, B=SkyAtmosphereLightIlluminance!(World Position))
  n7 = DotProduct(A=@n1, B=@n2)
  n8 = VectorParameter() {parameterName=Diffuse Color, defaultValue=<rgba 0.02 0.0333 0.04 9>}
  n9 = Multiply(A=Power(Base=ConstantBiasScale(None=@n7), Exp=@n8.A), B=@n8.RGB)
  n10 = Multiply(A=SkyAtmosphereDistantLightScatteredLuminance(), B=VectorParameter() {parameterName=Ambient, defaultValue=<rgba 0.027 0.021 0.03 0>}.RGB)
  FunctionOutput(None=Add(A=Multiply(A=@n5, B=@n6), B=Add(A=@n9, B=@n10)))
 ]}
