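# Material function that makes two looping, scaling UV sets and a blend alpha for seamless endless-zoom crossfades, stored with the Crystal Globe backpack FX.
# Time over Duration drives two half-cycle-offset phases, each scaling UVs around a pivot from start to end scale, interpolated in log10 space for exponential zoom.
# An optional static switch adds a noise-based random UV offset per cycle.
# A triangle-wave alpha hides each layer as it resets.
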MaterialFunction /BRCosmetics/Accessories/FORT_Backpacks/Backpack_M_MED_CrystalGlobe/FX/Materials/MF_ScaleCrossfadeUV
graph {nodes [
  n1 = FunctionInput(Preview=TextureCoordinate()) {bUsePreviewValueAsDefault=true, description="Defaults to TexCoord[0]", inputName=UV, inputType=FunctionInput_Vector2}
  n2 = FunctionInput(Preview=Time()) {bUsePreviewValueAsDefault=true, description=Defaults to GameTime, inputName=Time, inputType=FunctionInput_Scalar, sortPriority=5}
  n3 = FunctionInput() {
   bUsePreviewValueAsDefault=true
   description=How many seconds the cycle takes
   inputName=Duration
   inputType=FunctionInput_Scalar
   sortPriority=4
   previewValue=<3 0 0 1>
  }
  n4 = Divide(A=@n2, B=@n3)
  n5 = Floor(None=@n4)
  n6 = ComponentMask(None=VectorNoise(World Position=Add(A=@n5) {constB=0.5})) {g=true, r=true}
  n7 = FunctionInput() {
   bUsePreviewValueAsDefault=true
   description=How much randomness to apply to the UV offset
   inputName=Random UV Offset Scale
   inputType=FunctionInput_Vector2
   sortPriority=7
   previewValue=<1 1 0 1>
  }
  n8 = Constant()
  n9 = FunctionInput(Preview=StaticBool()) {
   bUsePreviewValueAsDefault=true
   description="Randomizes UV offset per cycle, adds a few instructions, defaults to FALSE"
   inputName=Random UV Offset Enable
   inputType=FunctionInput_StaticBool
   sortPriority=6
  }
  n10 = StaticSwitch(True=Multiply(A=Subtract(A=@n6) {constB=0.5}, B=@n7), False=@n8, Value=@n9)
  n11 = FunctionInput() {
   bUsePreviewValueAsDefault=true
   description="Where to center the scale. Defaults to 0.5, 0.5"
   inputName=UV Scale Pivot
   inputType=FunctionInput_Vector2
   sortPriority=1
   previewValue=<0.5 0.5 0 1>
  }
  n12 = Add(A=@n11, B=@n10)
  n13 = Subtract(A=Add(A=@n1, B=@n10), B=@n12)
  n14 = Constant() {r=10}
  n15 = FunctionInput() {
   bUsePreviewValueAsDefault=true
   description=UV Scale at the start of the cycle. Needs to be larger than zero
   inputName=UV Scale Start
   inputType=FunctionInput_Vector2
   sortPriority=2
   previewValue=<2 2 0 1>
  }
  n16 = Logarithm10!(X=@n15)
  n17 = FunctionInput() {
   bUsePreviewValueAsDefault=true
   description=UV Scale at the end of the cycle. Needs to be larger than zero
   inputName=UV Scale End
   inputType=FunctionInput_Vector2
   sortPriority=3
   previewValue=<0.5 0.5 0 1>
  }
  n18 = Logarithm10(X=@n17)
  n19 = Frac(None=@n4)
  n20 = Power(Base=@n14, Exp=LinearInterpolate(A=@n16, B=@n18, Alpha=@n19))
  FunctionOutput(None=Add(A=Multiply(A=@n13, B=@n20), B=@n12)) {description=Flipbook UV 1, outputName=UV 1}
  n21 = Add(A=@n4) {constB=0.5}
  n22 = Floor(None=@n21)
  n23 = ComponentMask(None=VectorNoise(World Position=Add(A=@n22) {constB=6.5})) {g=true, r=true}
  n24 = StaticSwitch(True=Multiply(A=Subtract(A=@n23) {constB=0.5}, B=@n7), False=@n8, Value=@n9)
  n25 = Add(A=@n11, B=@n24)
  n26 = Subtract(A=Add(A=@n1, B=@n24), B=@n25)
  n27 = Power(Base=@n14, Exp=LinearInterpolate(A=@n16, B=@n18, Alpha=Frac(None=@n21)))
  FunctionOutput(None=Add(A=Multiply(A=@n26, B=@n27), B=@n25)) {description=Flipbook UV 2, outputName=UV 2, sortPriority=1}
  FunctionOutput(None=Abs(None=ConstantBiasScale(None=@n19) {bias=-0.5, scale=2})) {description=Lerp Alpha for UV 1 and 2, outputName=Crossfade Alpha, sortPriority=3}
 ]}
