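# Unlit base material for the Crystal Globe backpack's goo effect on Niagara mesh particles.
# Wave-distorted, scale-crossfaded UVs blend two noise samples, whose alpha drives world position offset and RGB supplies the normal for custom goo shading.
# The vertex stage samples a blurrier mip to smooth displacement.
# A UV edge mask is built, and a particle "Grow" parameter adds offset along vertex normals.
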Material /BRCosmetics/Accessories/FORT_Backpacks/Backpack_M_MED_CrystalGlobe/FX/Materials/M_Backpack_CrystalGlobe_Goo_Base
graph {
 properties {
  bAutomaticallySetUsageInEditor=false
  bUsedWithNiagaraMeshParticles=true
  previewMesh=/BRCosmetics/Accessories/FORT_Backpacks/Backpack_M_MED_CrystalGlobe/FX/Meshes/SM_Backpack_CrystalGlobe_Goo_Base
  shadingModel=MSM_Unlit
  usedShadingModels=Unlit
 }
 nodes [
  n1 = ScalarParameter() {defaultValue=0.09, parameterName=UV Wave Amplitude}
  n2 = ScalarParameter() {defaultValue=0.3, parameterName=UV Wave Speed}
  n3 = VectorParameter() {parameterName=UV Scales, defaultValue=<rgba 0.5 1.4 1.5 1.333>}
  n4 = TextureCoordinate()
  n5 = MaterialFunctionCall!(
   UV=MaterialFunctionCall!(UV, U Frequency, V Frequency, U Amplitude=@n1, V Amplitude=@n1, U Speed=@n2, V Speed=@n2, U Time, V Time) {materialFunction=/Game/Effects/Fort_Effects/MaterialFunctions/MF_WaveUV}.Wave UVs
   UV Scale Pivot
   UV Scale Start=@n3.R
   UV Scale End=@n3.R
   Duration=@n3.R
   Time=Add(A=Time(), B=ComponentMask(None=@n4) {g=true})
   Random UV Offset Enable=StaticBool() {value=true}
   Random UV Offset Scale
  ) {materialFunction=&/MF_ScaleCrossfadeUV}
  n6 = NamedRerouteDeclaration(None=@n5.UV 1) {name=UV 1, nodeColor=<rgba 0 0.8125 1 1>}
  n7 = TextureObject() {
   samplerType=SAMPLERTYPE_LinearColor
   texture=/BRCosmetics/Accessories/FORT_Backpacks/Backpack_M_MED_CrystalGlobe/FX/Textures/T_Backpack_CrystalGlobe_Noise
  }
  n8 = TextureProperty(Tex=@n7)
  n9 = MaterialFunctionCall!(UVs=@n6, Texture Size=@n8) {materialFunction=/Engine/Functions/Engine_MaterialFunctions01/Texturing/ComputeMipLevel}
  n10 = ScalarParameter() {defaultValue=4.5, desc="This samples a blurrier version of the texture for the WPO, makes it less jagged", parameterName=WPO Mip Level}
  n11 = NamedRerouteDeclaration(None=@n5.UV 2) {name=UV 2, nodeColor=<rgba 0.4062 0 1 1>}
  n12 = MaterialFunctionCall(UVs=@n11, Texture Size=@n8) {materialFunction=&ComputeMipLevel}
  n13 = LinearInterpolate(
   A=TextureSample!(UVs=@n6, Tex=@n7, Level=ShaderStageSwitch(PixelShader=@n9.MipLevel, VertexShader=@n10), Apply View MipBias) {mipValueMode=TMVM_MipLevel, samplerType=SAMPLERTYPE_LinearColor}.RGBA
   B=TextureSample(UVs=@n11, Tex=@n7, Level=ShaderStageSwitch(PixelShader=@n12.MipLevel, VertexShader=@n10)) {mipValueMode=TMVM_MipLevel, samplerType=SAMPLERTYPE_LinearColor}.RGBA
   Alpha=NamedRerouteDeclaration(None=@n5.Crossfade Alpha) {name=Crossfade Alpha, nodeColor=<rgba 0.0078 0 1 1>}
  )
  n14 = Power(Base=OneMinus(None=ComponentMask(None=@n13) {a=true}), Exp=@n3.B)
  n15 = Multiply(
   A=ConstantBiasScale(None=OneMinus(None=@n14)) {bias=-0.5, scale=2}
   B=Multiply(A=ScalarParameter() {defaultValue=5, parameterName=WPO Intensity}, B=NamedRerouteUsage())
  )
  n16 = LinearInterpolate(
   A=Transform(None=Constant3Vector() {constant=<rgba 0 0 1 0>}) {transformSourceType=TRANSFORMSOURCE_Instance}
   B=VertexNormalWS()
   Alpha=NamedRerouteUsage()
  )
  n17 = VertexNormalWS()
  n18 = NamedRerouteUsage()
  Add(
   A=NamedRerouteDeclaration(None=Multiply(A=@n15, B=@n16)) {name=World Position Offset, nodeColor=<rgba 0 0.4375 1 1>}
   B=Multiply(A=Multiply(A=@n17, B=@n18), B=DynamicParameter() {paramNames [Grow,Param2,Param3,Param4]}.Grow) {constB=-4}
  )
  n19 = ConstantBiasScale(None=TextureCoordinate()) {bias=-0.5, scale=2}
  n20 = MaterialFunctionCall(Vector 2=Power(Base=Abs(None=@n19), Exp=ScalarParameter() {defaultValue=3, parameterName=Edges Mask Exponent}) {constExponent=3}) {materialFunction=/Engine/Functions/Engine_MaterialFunctions02/Utility/VectorLength}
  NamedRerouteDeclaration(None=Saturate(None=OneMinus(None=@n20.V2 Length))) {name=Edges Mask, nodeColor=<rgba 0 0.3672 1 1>}
  n21 = ConstantBiasScale(None=ComponentMask(None=@n13) {b=true, g=true, r=true}) {bias=-0.5, scale=2}
  n22 = LinearInterpolate(B=ScalarParameter() {defaultValue=0.5, parameterName=Normal Flatness}, Alpha=NamedRerouteUsage()) {constA=1}
  n23 = Transform(None=Normalize(VectorInput=MaterialFunctionCall(Normal=@n21, Flatness=@n22) {materialFunction=&FlattenNormal}.Result))
  NamedRerouteDeclaration(None=MaterialFunctionCall(Emissive=MaterialFunctionCall!(WS Normal=@n23) {materialFunction=&/MF_Backpack_CrystalGlobe_Goo_Shading}.Result) {materialFunction=&MF_FX_EmissiveAdjustments}.Result) {name=Emissive Color, nodeColor=<rgba 1 0 0.0469 1>}
 ]
}
