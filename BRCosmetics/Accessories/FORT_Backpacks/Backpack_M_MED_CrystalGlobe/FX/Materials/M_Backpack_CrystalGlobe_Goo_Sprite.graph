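# Masked unlit Niagara sprite material for the goo effect on the Crystal Globe backpack.
# MF_ScaleCrossfadeUV blends two samples of a noise texture over randomly offset time, with UVs scaled by particle size.
# Noise alpha feeds MF_Erosion, with erosion the maximum of a powered edge mask and a dynamic parameter.
# Flattened noise RGB plus a derived edge normal feeds the goo shading function for emissive; both edge axes use the U exponent.
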Material /BRCosmetics/Accessories/FORT_Backpacks/Backpack_M_MED_CrystalGlobe/FX/Materials/M_Backpack_CrystalGlobe_Goo_Sprite
graph {
 properties {
  bAutomaticallySetUsageInEditor=false
  bUsedWithNiagaraSprites=true
  blendMode=BLEND_Masked
  shadingModel=MSM_Unlit
  translucencyPass=MTP_BeforeDOF
  usedShadingModels=Unlit
 }
 nodes [
  n1 = TextureCoordinate()
  n2 = ParticleSize()
  n3 = ScalarParameter() {defaultValue=0.0666, parameterName=UV Scale}
  n4 = VectorParameter() {parameterName=UV Scales, defaultValue=<rgba 1 0.75 0.6 0>}
  n5 = MaterialFunctionCall!(
   UV=Multiply(A=Subtract(A=@n1) {constB=0.5}, B=Multiply(A=@n2, B=@n3))
   UV Scale Pivot=Constant()
   UV Scale Start=@n4.R
   UV Scale End=@n4.R
   Duration=@n4.R
   Time=Add(A=Time(), B=DynamicParameter() {paramNames [Edges Mask Exponent U, Edges Mask Exponent V, Erosion, Random Time Offset]}.Random Time Offset)
   Random UV Offset Enable=StaticBool() {value=true}
   Random UV Offset Scale
  ) {materialFunction=&/MF_ScaleCrossfadeUV}
  n6 = TextureObject() {
   samplerType=SAMPLERTYPE_LinearColor
   texture=/BRCosmetics/Accessories/FORT_Backpacks/Backpack_M_MED_CrystalGlobe/FX/Textures/T_Backpack_CrystalGlobe_Noise
  }
  n7 = TextureSample(UVs=@n5.UV 1, Tex=@n6) {
   samplerType=SAMPLERTYPE_LinearColor
   texture=&T_Backpack_CrystalGlobe_Noise
  }
  n8 = TextureSample(UVs=@n5.UV 2, Tex=@n6) {
   samplerType=SAMPLERTYPE_LinearColor
   texture=&T_Backpack_CrystalGlobe_Noise
  }
  n9 = LinearInterpolate(A=@n7.RGBA, B=@n8.RGBA, Alpha=@n5.Crossfade Alpha)
  n10 = ConstantBiasScale(None=TextureCoordinate()) {bias=-0.5, scale=2}
  n11 = DynamicParameter() {paramNames [Edges Mask Exponent U, Edges Mask Exponent V, Erosion, Random Time Offset]}
  n12 = Power(Base=Abs(None=@n10), Exp=AppendVector(A=@n11.Edges Mask Exponent U, B=@n11.Edges Mask Exponent U)) {constExponent=3}
  n13 = MaterialFunctionCall(Vector 2=@n12) {materialFunction=/Engine/Functions/Engine_MaterialFunctions02/Utility/VectorLength}
  MaterialFunctionCall!(
   Gradient=ComponentMask(None=@n9) {a=true}
   Erosion Amount=Max(
    A=Saturate(None=@n13.V2 Length)
    B=DynamicParameter() {paramNames [Edges Mask Exponent U, Edges Mask Exponent V, Erosion, Random Time Offset]}.Erosion
   )
   Softness=ScalarParameter() {defaultValue=0.05, parameterName=Erosion Softness}
  ) {materialFunction=/BRCosmetics/Athena/Effects/Particles/Skydiving/NS_Contrail_DarkAndLight/Functions/MF_Erosion}
  n14 = ConstantBiasScale(None=ComponentMask(None=@n9) {b=true, g=true, r=true}) {bias=-0.5, scale=2}
  n15 = ScalarParameter() {defaultValue=0.4, parameterName=Normal Flatness}
  n16 = Normalize(VectorInput=Add(
    A=MaterialFunctionCall(Normal=@n14, Flatness=@n15) {materialFunction=&FlattenNormal}.Result
    B=NamedRerouteUsage()
   ))
  MaterialFunctionCall(Emissive=MaterialFunctionCall!(WS Normal=Transform(None=@n16)) {materialFunction=&/MF_Backpack_CrystalGlobe_Goo_Shading}.Result) {materialFunction=&MF_FX_EmissiveAdjustments}
  n17 = Multiply(A=@n12, B=Sign(None=@n10))
  n18 = ScalarParameter() {defaultValue=2, parameterName=Edge Normal Intensity}
  NamedRerouteDeclaration(None=DeriveNormalZ(InXY=Multiply(A=@n17, B=@n18))) {name=Edge Normal, nodeColor=<rgba 0.0312 1 0 1>}
 ]
}
