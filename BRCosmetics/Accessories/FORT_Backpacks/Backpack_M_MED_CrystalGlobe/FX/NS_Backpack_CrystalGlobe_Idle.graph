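# Niagara idle effect for the Crystal Globe backpack cosmetic, showing animated goo, rising strings and beads, and a red glow.
# BaseGooMesh ramps a material parameter with system age.
# StringsAndBeads uses two spawn groups for differently sized, slowly rising sprites, killed outside a box.
# LightAndGlow fades in glow sprites and a light tinted by time of day, with a front-end color variant and a quality-gated extra spawn.
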NiagaraSystem /BRCosmetics/Accessories/FORT_Backpacks/Backpack_M_MED_CrystalGlobe/FX/NS_Backpack_CrystalGlobe_Idle
dependencies {
 usedDataInterfaces [
  &NiagaraDataInterfaceArrayFloat2
  &NiagaraDataInterfaceCurve
  &NiagaraDataInterfaceVector2DCurve
  &NiagaraDataInterfaceArrayInt32
 ]
 usedDynamicInputs [
  &SelectVector2DFromArray
  &SpawnGroup
  &MakeVectorFromLinearColorRGB
 ]
 usedModules [&SystemState]
}
emitters {
 BaseGooMesh {
  emitterUpdateScript [
   &EmitterState(
    Inactive Response=::NewEnumerator1
    Life Cycle Mode=::NewEnumerator1
    Loop Behavior=::NewEnumerator1
    Loop Duration Mode=::NewEnumerator1
    UseLoopDelay=false
   )
   &SpawnBurst_Instantaneous(Use Spawn Probability=false, UseLoopCountLimit=false)
  ]
  particleSpawnScript [&InitializeParticle(
    Color=<rgba 3 0.5128 0 1>
    Mesh Scale Mode=::NewEnumerator4
    Sprite Size Mode=::NewEnumerator3
    Uniform Sprite Size=37
    UsePositionOffset=false
    Write Material Random=true
   )]
  particleUpdateScript [
   &ParticleState(Kill Particles When Lifetime Has Elapsed=false, Let Infinitely Lived Particles Die When Emitter Deactivates=true)
   &DynamicMaterialParameters(
    Index 0 Param 1=&FloatFromCurve(
     FloatCurve=DataInterface() {propertyValues {Curve {keys [
         {arriveTangent=2.7135, interpMode=RCIM_Cubic, leaveTangent=2.7135, tangentMode=RCTM_User, time=0.2, value=-8}
         {arriveTangent=12.8712, interpMode=RCIM_Cubic, leaveTangent=12.8712, tangentMode=RCTM_User, time=0.6, value=-4}
         {arriveTangent=0.4423, tangentMode=RCTM_User, time=1.2, value=0}
        ]}}}
     CurveIndex=$Engine.System.Age
    )
    Module.Index0Param0Write=true
    Module.Index0Param1Write=true
    Module.Index0Param2Write=true
    Module.Index0Param3Write=true
   )
  ]
  rendererClasses [&NiagaraMeshRendererProperties]
  renderers [{rendererClass=&NiagaraMeshRendererProperties, rendererIndex=0}]
 }
 LightAndGlow {
  emitterUpdateScript [
   &EmitterState(
    Inactive Response=::NewEnumerator1
    Life Cycle Mode=::NewEnumerator1
    Loop Behavior=::NewEnumerator1
    Loop Duration Mode=::NewEnumerator1
    UseLoopDelay=false
   )
   &SpawnBurst_Instantaneous(Spawn Count=3, Use Spawn Probability=false, UseLoopCountLimit=false)
   &SpawnBurst_Instantaneous(
    Spawn Count=&SelectIntFromArray(Direct Array Index=$Engine.QualityLevel, Int Selection Array=DataInterface() {propertyValues {IntData [0,0,1,1,1]}})
    Spawn Group=1
    Use Spawn Probability=false
    UseLoopCountLimit=false
   ) {module=SpawnBurst_Instantaneous001}
  ]
  particleSpawnScript [
   /SetVariables(Particles.VisibilityTag=$Emitter.SpawnGroup) {isSetParameters=true}
   &InitializeParticle(
    Lifetime=5
    Position Offset=<0 -12.4496 -2.0854>
    Sprite Size=<44.2775 1>
    Sprite Size Mode=::NewEnumerator4
    UsePositionOffset=true
    Write Material Random=true
   )
   &CameraOffset(Camera Offset Amount=-34.0028)
  ]
  particleUpdateScript [
   &ParticleState(Kill Particles When Lifetime Has Elapsed=false)
   &DynamicMaterialParameters(
    Index 0 Param 1=&MakeFloatFromVector(Vector=$Engine.Owner.SystemZAxis, Channel=::NewEnumerator0)
    Index 0 Param 2=&MakeFloatFromVector(Vector=$Engine.Owner.SystemZAxis, Channel=::NewEnumerator1)
    Index 0 Param 3=&MakeFloatFromVector(Vector=$Engine.Owner.SystemZAxis, Channel=::NewEnumerator2)
    Index 0 Param 4=66.5184
    Module.Param0WriteEnabled=true
    Module.Param1WriteEnabled=true
    Module.Param2WriteEnabled=true
    Module.Param3WriteEnabled=true
   )
   &ScaleColor(
    Scale Mode=::NewEnumerator1
    Scale RGBA=&MakeVector4_FromLinearColor(LinearColor=&FNTimeOfDayColor(
      Night Color=<rgba 1.15 0 0 0.45>
      Morning Color=<rgba 1.5 0 0 0.5>
      Day Color=&MakeCustomLinearColorFromBool(
       Bool=&Boolean_OrOperation(Bool1=$User.bIsFrontEnd, Bool2=$User.bIsFrontEndPreview)
       True Color=<rgba 1.3 0 0 0.4>
       False Color=<rgba 1.25 0 0 0.8>
      )
      Evening Color=<rgba 1.25 0 0 0.45>
     ))
    ScaleRGBA=true
   ) {module=ScaleColor001}
   &ScaleColor(
    Color Value To Scale=$Particles.Color
    Scale Alpha=&FloatFromCurve(
     FloatCurve=DataInterface() {propertyValues {Curve {keys [{tangentMode=RCTM_User, time=0, value=0}, {tangentMode=RCTM_User, time=1, value=1}]}}}
     CurveIndex=$Engine.System.Age
    )
    ScaleA=true
    ScaleRGB=true
   )
   &FN_Light_Attributes(
    Color=&FNTimeOfDayColor(Night Color=<rgba 1.15 0 0 0.45>, Morning Color=<rgba 1.5 0 0 0.5>, Day Color=<rgba 2 0 0 0.5>, Evening Color=<rgba 1.25 0 0 0.45>)
    Create Color Attribute=true
    Create Position Attribute=true
    Create Radius Attribute=true
    Position=<0 -20 -1.9182>
    Radius=14.2412
    Scale Color=&FloatFromCurve(
     FloatCurve=DataInterface() {propertyValues {Curve {keys [{tangentMode=RCTM_User, time=0, value=0}, {tangentMode=RCTM_User, time=1, value=1}]}}}
     CurveIndex=$Engine.System.Age
    )
   )
  ]
  rendererClasses [&NiagaraSpriteRendererProperties, &NiagaraLightRendererProperties]
  renderers [
   {rendererClass=&NiagaraSpriteRendererProperties}
   {rendererClass=&NiagaraLightRendererProperties, rendererIndex=1}
  ]
 }
 StringsAndBeads {
  emitterUpdateScript [
   &EmitterState(
    Delay First Loop Only=true
    Inactive Response=::NewEnumerator1
    Life Cycle Mode=::NewEnumerator1
    Loop Delay=0.2
    Loop Duration=1
    UseLoopDelay=true
   )
   &SpawnRate(SpawnRate=3, Use Spawn Probability=false) {module=SpawnRate001}
   &SpawnRate(SpawnGroup=1, SpawnRate=2, Use Spawn Probability=false)
  ]
  particleSpawnScript [
   &InitializeParticle(
    Color Mode=::NewEnumerator0
    Lifetime=&MaskFloatBySpawnGroup(
     Masked Float=&RandomRangeFloat(Minimum=3, Maximum=3)
     Passthrough Float=&RandomRangeFloat(Minimum=1.6, Maximum=2.8)
    )
    Sprite Size=&MaskVector2dBySpawnGroup(
     Masked Vector2D=&Vector2DFromFloat(Value=&RandomRangeFloat(Minimum=2, Maximum=4))
     Passthrough Vector2D=&RandomRangeVector2D(Minimum=<3 11>, Maximum=<5 22>)
     Spawn Group Mask=0
    )
    Sprite Size Mode=::NewEnumerator4
    UsePositionOffset=false
    Write Material Random=true
   )
   &ShapeLocation(
    Box Size=<8 8 0>
    MaskSpawnGroup=false
    Offset=<0 -12 -12>
    Offset Mode=::NewEnumerator0
    Override Seed=false
    Shape Primitive=::NewEnumerator2
    Surface Only=false
    UseOwnerScale=true
   )
   &AddVelocity(
    Velocity=<0 0 7>
    Velocity Speed Scale=&MaskFloatBySpawnGroup(Masked Float=&RandomRangeFloat(Minimum=0.5, Maximum=1), Passthrough Float=0.2)
   )
   &ApplyInitialForces()
   &SpriteFacingAndAlignment(Sprite Alignment=<0 0 1>, WriteAlignment=true, WriteFacing=false)
  ]
  particleUpdateScript [
   &ParticleState()
   &ScaleSpriteSize(
    Non-Uniform Curve Sprite Scale=DataInterface() {propertyValues {
      XCurve {keys [<at 0.8 1 RCIM_Cubic>, {arriveTangent=0.0224, interpMode=RCIM_Cubic, leaveTangent=0.0224, tangentMode=RCTM_User, time=1, value=0}]}
      YCurve {keys [{interpMode=RCIM_Cubic, tangentMode=RCTM_User, time=0, value=0}, {interpMode=RCIM_Cubic, tangentMode=RCTM_User, time=1, value=1}]}
     }}
    Scale Sprite Size Mode=::NewEnumerator3
   )
   &AccelerationForce(Acceleration=&Multiply_VectorByFloat(
     Vector=&RandomRangeVector(Minimum=<-1 -1 0>, Maximum=<1 1 0>)
     Float=&RandomRangeFloat(Minimum=0, Maximum=2)
    ))
   &SolveForcesAndVelocity(Clamp Velocity=false, Limit Acceleration=false)
   &DynamicMaterialParameters(
    Index 0 Param 1=&MaskFloatBySpawnGroup(
     Masked Float=1
     Passthrough Float=&Lerp_Float(
      A=1.5
      B=0.5
      Alpha=&FloatFromCurve(FloatCurve=DataInterface() {propertyValues {Curve {keys [{interpMode=RCIM_Cubic, tangentMode=RCTM_User, time=0, value=0}, {interpMode=RCIM_Cubic, tangentMode=RCTM_User, time=1, value=1}]}}})
     )
    )
    Index 0 Param 2=&MaskFloatBySpawnGroup(Masked Float=1, Passthrough Float=9)
    Index 0 Param 3=&FloatFromCurve(FloatCurve=DataInterface() {propertyValues {Curve {keys [<at 0 0>, {arriveTangent=1.4993, time=0.4152, value=0.014}, <at 0.8228 0.3621>, {arriveTangent=6.8718, time=1, value=1}]}}})
    Index 0 Param 4=&RandomRangeFloat(Minimum=0, Maximum=33)
    Module.Index0Param0Write=true
    Module.Index0Param1Write=true
    Module.Index0Param2Write=true
    Module.Index0Param3Write=true
   )
   /SetVariables(Particles.SpriteSize=&MaskVector2dBySpawnGroup(Masked Vector2D=$Particles.Initial.SpriteSize, Passthrough Vector2D=$Particles.SpriteSize, Spawn Group Mask=0)) {isSetParameters=true}
   &KillParticlesInVolume(Box Size:Vector3f=<11.5 8 40>, Invert Volume=true, Kill Shape=::NewEnumerator1, Origin Offset:Vector3f=<0 -12 -12>)
  ]
  rendererClasses [&NiagaraSpriteRendererProperties]
  renderers [{rendererClass=&NiagaraSpriteRendererProperties}]
 }
}
summary {
 systemName=NS_Backpack_CrystalGlobe_Idle
 emitters [
  {emitterName=BaseGooMesh, rendererClasses [&NiagaraMeshRendererProperties]}
  {emitterName=StringsAndBeads, rendererClasses [&NiagaraSpriteRendererProperties]}
  {emitterName=LightAndGlow, rendererClasses [&NiagaraSpriteRendererProperties, &NiagaraLightRendererProperties]}
 ]
 userVariables [
  User.bIsFrontEnd {type=NiagaraBool}
  User.bIsFrontEndPreview {type=NiagaraBool}
 ]
}
